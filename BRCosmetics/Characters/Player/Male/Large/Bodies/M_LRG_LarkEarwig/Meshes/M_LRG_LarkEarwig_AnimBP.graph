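# Animation Blueprint for the M_LRG_LarkEarwig character body part, which follows the attached parent mesh and adds secondary motion.
# The copied pose runs through the body's Control Rig, a backpack attach-bone offset, then spring-based AnimDynamics on both boot bones.
# Skin weight profile and arm visibility events set the rig's retargeting and gauntlet flags.
# Updates also zero a velocity-scale variable while ziplining, though the dynamics nodes show fixed scales.
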AnimBlueprint /BRCosmetics/Characters/Player/Male/Large/Bodies/M_LRG_LarkEarwig/Meshes/M_LRG_LarkEarwig_AnimBP
target_skeleton=&/M_LRG_LarkEarwig_Skeleton
graphs [
 AnimGraph {nodes [
   n1 = "Transform (Modify) Bone - Bone: attach_backpack"(
    ComponentPose=Local To Component(LocalPose=Use cached pose 'ControlRig_CACHE'())
    Alpha=1
    Translation="0,-7,0"
    Rotation="0,0,0"
    Scale="1,1,1"
   ) {translationMode=BMM_Additive, translationSpace=BCS_BoneSpace, boneToModify {boneName=attach_backpack}, translation {y=-7}}
   n2 = Anim Dynamics - Bone: dyn_boot_lf!(ComponentPose=@n1, Alpha=1, ComponentLinearAccScale:Vector="1,1,1", ComponentLinearVelScale:Vector="1,1,1", ExternalForce="0,0,0") -> (Pose) {
    angularDampingOverride=1
    angularSpringConstant=20000
    bAngularSpring=true
    bOverrideAngularDamping=true
    bUseGravityOverride=true
    gravityScale=0.35
    simulationSpace=RootRelative
    worldSpaceMinimumScale=0.7
    boundBone {boneName=dyn_boot_lf}
    componentAppliedLinearAccClamp=<3000 3000 3000>
    componentLinearAccScale=<1 1 1>
    componentLinearVelScale=<1 1 1>
    gravityOverride {z=-980}
    physicsBodyDefinitions [{
      boundBone {boneName=dyn_boot_lf}
      boxExtents=<70 70 70>
      constraintSetup {angularLimitsMax=<0 2 5>, angularLimitsMin=<0 -2 -2>, angularTarget=<1 0 0>}
      localJointOffset=<1 0 0>
     }]
    simSpaceSettings {maxAngularAcceleration=35, maxAngularVelocity=7, simSpaceAngularAlpha=1}
   }
   n3 = Anim Dynamics - Bone: dyn_boot_rt!(ComponentPose=@n2, Alpha=1, ComponentLinearAccScale:Vector="1,1,1", ComponentLinearVelScale:Vector="1,1,1", ExternalForce="0,0,0") -> (Pose) {
    angularDampingOverride=1
    angularSpringConstant=20000
    bAngularSpring=true
    bOverrideAngularDamping=true
    bUseGravityOverride=true
    gravityScale=0.35
    simulationSpace=RootRelative
    worldSpaceMinimumScale=0.7
    boundBone {boneName=dyn_boot_rt}
    componentAppliedLinearAccClamp=<3000 3000 3000>
    componentLinearAccScale=<1 1 1>
    componentLinearVelScale=<1 1 1>
    gravityOverride {z=-980}
    physicsBodyDefinitions [{
      boundBone {boneName=dyn_boot_rt}
      boxExtents=<70 70 70>
      constraintSetup {angularLimitsMax=<0 2 5>, angularLimitsMin=<0 -2 -2>, angularTarget=<1 0 0>}
      localJointOffset=<1 0 0>
     }]
    simSpaceSettings {maxAngularAcceleration=35, maxAngularVelocity=7, simSpaceAngularAlpha=1}
   }
   Output Pose(Result=Component To Local(ComponentPose=@n3)) {layerGroup=None}
   n4 = Control Rig!(
    Source=Copy Pose From Mesh() {bUseAttachedParent=true}
    Alpha
    Get Disable Retargeting:bool=Get Disable Retargeting() -> (Disable Retargeting:bool)
    Get Is Wearing Gauntlets:bool=Get isWearingGauntlets() -> (isWearingGauntlets:bool)
   ) -> (Pose) {controlRigAssetReference {blueprintRigClass=&/M_LRG_LarkEarwig_CtrlRig.M_LRG_LarkEarwig_CtrlRig_C}}
   Save cached pose 'ControlRig_CACHE'(Pose=@n4) {cacheName=ControlRig_CACHE}
  ]}
 EventGraph {nodes [
   Event BlueprintUpdateAnimation()
   > n1 = Sequence()
   n2 = GetOwnerAnimationInterface!(self:Custom Character Part Anim Instance Object Reference) -> (ReturnValue:Fort Player Animation Interface Object Reference)
   > Is Valid(InputObject=@n2)
   > n3 = Branch(Condition=Break Fort Anim Input Zipline!(FortAnimInput_Zipline:"Fort Anim Input Zipline Structure (by ref)"=GetZiplineInput!(self:Fort Player Animation Interface Object Reference=@n2) -> (ReturnValue:"Fort Anim Input Zipline Structure (by ref)")) -> (
     bIsZiplining:bool=~
     bIsOnAscender:bool=~
     bShouldPlayPivotTransition:bool=~
     LeanYaw:float=~
     PivotBlendDelayRemaining:float=~
     PivotBlendDelay:float=~
     PivotCardinalDirection:EFortCardinalDirection Enum
     WorldVelocityLastTick:Vector=~
    ) {is=title K2Node_BreakStruct})
   > Set Componen Linear Vel Scale(Componen Linear Vel Scale:Vector="0, 0, 0") -> (Output_Get:Vector)
   Set Componen Linear Vel Scale(execute=@n3.else, Componen Linear Vel Scale="0.3,0.3,0.3")
   n4 = GetPartSkeletalMeshComponent()
   Set Disable Retargeting(
    execute=@n1.then_1
    Disable Retargeting:bool="Not Equal (Name)"(A=GetCurrentSkinWeightProfileName!(self:Skinned Mesh Component Object Reference=@n4, InLayer:ESkinWeightProfileLayer Enum=Primary) -> (ReturnValue:Name=None)) -> (ReturnValue=~)
   ) -> (Output_Get:bool)
   n5 = Event OnLeftArmHidden() {is=title K2Node_Event}
   Set isWearingGauntlets(execute=[@n5.then, @n8.then], isWearingGauntlets:bool=true) -> (Output_Get:bool)
   n6 = Event OnLeftArmVisible() {is=title K2Node_Event}
   Set isWearingGauntlets(execute=[@n6.then, @n7.then], isWearingGauntlets=false)
   n7 = Event OnRightArmVisible() {is=title K2Node_Event}
   n8 = Event OnRightArmHidden() {is=title K2Node_Event}
  ]}
]
variables [
 Componen Linear Vel Scale {type=Vector, default=<0.3 0.3 0.3>}
 Disable Retargeting {type=boolean}
 isWearingGauntlets {type=boolean}
]
