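# Physics asset for the face accessory mesh of the large male LarkEarwig player cosmetic, setting up physics for four hair sections: back center, front center, front left and front right.
# Each section pairs a tiny capsule on a base bone with a larger, offset capsule on a dyn bone.
# Each dyn bone is constrained only to its own base, with twist limited to zero and one swing axis much tighter than the other.
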PhysicsAsset /BRCosmetics/Characters/Player/Male/Large/Bodies/M_LRG_LarkEarwig/Meshes/Parts/M_LRG_LarkEarwig_FaceAcc_Physics
bodies {
 base_hair_back_cr [{shapeType=Capsule, length=0.505, radius=0.505, shapeName=base_hair_back_cr_capsule, rotation=<rot 0 -90 0>}]
 base_hair_front_cr [{shapeType=Capsule, length=0.505, radius=0.505, shapeName=base_hair_front_cr_capsule, rotation=<rot 0 -90 0>}]
 base_hair_front_lf [{shapeType=Capsule, length=0.505, radius=0.505, shapeName=base_hair_front_lf_capsule, rotation=<rot 0 -90 0>}]
 base_hair_front_rt [{shapeType=Capsule, length=0.505, radius=0.505, shapeName=base_hair_front_rt_capsule, rotation=<rot 0 -90 0>}]
 dyn_hair_back_cr [{shapeType=Capsule, length=3, radius=2, shapeName=dyn_hair_back_cr_capsule, center=<10 0 0>, rotation=<rot 90 0 0>}]
 dyn_hair_front_cr [{shapeType=Capsule, length=3, radius=2, shapeName=dyn_hair_front_cr_capsule, center=<10 0 0>, rotation=<rot 90 0 0>}]
 dyn_hair_front_lf [{shapeType=Capsule, length=3, radius=2, shapeName=dyn_hair_front_lf_capsule, center=<10 0 0>, rotation=<rot 90 0 0>}]
 dyn_hair_front_rt [{shapeType=Capsule, length=3, radius=2, shapeName=dyn_hair_front_rt_capsule, center=<10 0 0>, rotation=<rot 90 0 0>}]
}
constraints [
 {bone1Name=dyn_hair_back_cr, bone2Name=base_hair_back_cr, swing1LimitDegrees=5, swing2LimitDegrees=30, twistLimitDegrees=0}
 {bone1Name=dyn_hair_front_cr, bone2Name=base_hair_front_cr, swing1LimitDegrees=5, swing2LimitDegrees=30, twistLimitDegrees=0}
 {bone1Name=dyn_hair_front_lf, bone2Name=base_hair_front_lf, swing1LimitDegrees=5, swing2LimitDegrees=30, twistLimitDegrees=0}
 {bone1Name=dyn_hair_front_rt, bone2Name=base_hair_front_rt, swing1LimitDegrees=5, swing2LimitDegrees=30, twistLimitDegrees=0}
]
